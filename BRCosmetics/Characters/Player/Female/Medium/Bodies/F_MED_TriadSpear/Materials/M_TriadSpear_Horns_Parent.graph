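# Parent material for the TriadSpear character's horns, built on the 2023 character parent PreFX and PostFX functions with an audio-reactive emissive glow.
# Emissive lerps between a dimmed and a boosted horn texture, the latter scaled by a rerouted value.
# Blend alpha is powered frontend-loop audio amplitude, falling back to cylindrically projected panning noise without positive amplitude.
# Two mirrored sphere masks with variable radii combine into a Horns_L reroute, alongside a pre-skinned position reroute.
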Material /BRCosmetics/Characters/Player/Female/Medium/Bodies/F_MED_TriadSpear/Materials/M_TriadSpear_Horns_Parent
graph {
 properties {
  bAutomaticallySetUsageInEditor=false
  bUseMaterialAttributes=true
  bUsedWithMeshDeformer=true
  bUsedWithMorphTargets=true
  bUsedWithSkeletalMesh=true
  pixelDepthOffsetMode=PDOM_Legacy
  shadingModel=MSM_FromMaterialExpression
  usedShadingModels=Default Lit | Subsurface | Clear Coat | Subsurface Profile | Cloth
 }
 nodes [
  Constant()
  Constant() {r=1}
  n1 = MaterialFunctionCall() {materialFunction=&MF_CharacterParent_2023_PreFX}
  n2 = TextureSample() {texture=/BRCosmetics/Characters/Player/Female/Medium/Bodies/F_MED_TriadSpear/Textures/T_TriadSpear_HairHorns_Horns_E2}
  n3 = Multiply(A=@n2.RGB) {constB=0.2}
  n4 = Multiply(A=NamedRerouteUsage(), B=Multiply(A=@n2.RGB, B=Constant() {r=3}) {constB=3})
  n5 = Power(
   Base=MaterialFunctionCall!(InAmplitude, InSpectrum, CustomFakeSpectrum, CustomFakeAmplitude) {materialFunction=/BRCosmetics/Global/MaterialFunctions/AudioReactivity/MF_AudioReactivity_FrontendLoop}.Amplitude
   Exp=Constant() {r=3}
  ) {constExponent=4}
  n6 = Constant() {r=0.1}
  n7 = Constant() {r=-7.1828}
  n8 = TextureSample(UVs=Panner(Coordinate=MaterialFunctionCall!(Tile=@n6, DistortionAdjustment=@n7) {materialFunction=/BRCosmetics/Characters/Player/Female/Medium/Bodies/F_MED_TriadSpear/Materials/Body/MF_CylindricalProjUV}.UVs) {speedX=0.2, speedY=0.7}) {
   samplerSource=SSM_Wrap_WorldGroupSettings
   samplerType=SAMPLERTYPE_Grayscale
   texture=/Game/Global/Textures/Default/Noise/LowResBlurredNoiseNormalized
  }
  n9 = Multiply(A=@n8.R, B=Constant() {r=0.9})
  n10 = NamedRerouteDeclaration(None=If(A=@n5, "A > B"=@n5, "A == B"=@n9, "A < B"=@n9)) {name=In_Ampltiude, nodeColor=<rgba 1 0.2812 0 1>}
  MaterialFunctionCall(
   Material=SetMaterialAttributes!(
    MaterialAttributes=GetMaterialAttributes(None=@n1.Result) {attributeGetTypes [#1]}.MaterialAttributes
    Base Color
    Emissive Color=LinearInterpolate(A=@n3, B=@n4, Alpha=@n10)
   ) {attributeSetTypes [#2, #1]}
   M=@n1.M
   Pre Skinned Local Position=@n1.M
   Distance Blend=@n1.M
   Fresnel=@n1.M
  ) {materialFunction=&MF_CharacterParent_2023_PostFX}
  n11 = NamedRerouteUsage()
  n12 = Constant4Vector() {constant=<rgba -6.0938 -1.118 180 1>}
  n13 = LinearInterpolate(A=Constant() {r=10}, B=Constant() {r=20}, Alpha=NamedRerouteUsage()) {constB=3}
  n14 = Constant() {r=0.8}
  n15 = NamedRerouteUsage()
  n16 = Constant4Vector() {constant=<rgba 6.0938 -1.118 180 1>}
  n17 = LinearInterpolate(A=Constant() {r=10}, B=Constant() {r=20}, Alpha=NamedRerouteUsage()) {constB=3}
  n18 = Constant() {r=0.8}
  NamedRerouteDeclaration(None=Max(A=SphereMask(A=@n11, B=@n12.RGB, Radius=@n13, Hardness=@n14), B=SphereMask(A=@n15, B=@n16.RGB, Radius=@n17, Hardness=@n18))) {name=Horns_L, nodeColor=<rgba 0.8984 1 0 1>}
  NamedRerouteDeclaration(None=@n1.Pre skinned Local Position) {name=Preskin, nodeColor=<rgba 0.4531 0 1 1>}
  NamedRerouteUsage()
 ]
}
