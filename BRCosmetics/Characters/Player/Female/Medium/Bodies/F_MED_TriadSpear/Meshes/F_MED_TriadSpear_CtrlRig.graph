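# Control Rig for the F_MED_TriadSpear female medium body that applies corrective deformation and clothing bone adjustments.
# Transform offsets and a shared DeformRig function run first, then weighted parent constraints and offsets adjust foot deform and T-shirt base bones.
# Thigh rotation against reference nulls, remapped and scaled by IsIdle, rotates boot zipper bones.
# A left-thigh spherical pose reader, gated off by Moving2D and smoothed, weights glute offsets.
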ControlRigBlueprint /BRCosmetics/Characters/Player/Female/Medium/Bodies/F_MED_TriadSpear/Meshes/F_MED_TriadSpear_CtrlRig
graphs [
 New Function_ContainedGraph {nodes [
   n1 = RigVMFunctionEntryNode()
   RigVMFunction_MathQuaternionToEuler = &RigVMFunction_MathQuaternionToEuler(
    RotationOrder=XYZ
    Value=&RigUnit_GetRelativeTransformForItem(
     Child=(Type=Bone,Name="hand_l")
     Parent=(Type=Null,Name="lowerarm_l")
     bChildInitial=False
     bParentInitial=False
     Child.Name=@n1.Child
     Parent.Name=@n1.Parent
    ).RelativeTransform.Rotation
   )
   Remap = &RigVMFunction_MathDoubleRemap(
    TargetMaximum=1
    TargetMinimum=0
    Value=@RigVMFunction_MathQuaternionToEuler.Result.Y
    SourceMaximum=@n1.Source Maximum
    SourceMinimum=@n1.Source Minimum
   )
   > RigVMFunctionReturnNode(Delta Rotation=@Remap.Result)
  ]}
 RigVMFunctionLibrary {nodes [Delta Rotation X = RigVMCollapseNode(ExecuteContext=())]}
 RigVMModel {nodes [
   RigUnit_BeginExecution = &RigUnit_BeginExecution()
   &RigUnit_ModifyTransforms(
    ItemToModify="((Item=(Type=Bone,Name=\"upperarm_l\"),Transform=(Rotation=(X=0,Y=0,Z=0,W=1),Translation=(X=0,Y=0,Z=0),Scale3D=(X=1,Y=1,Z=1))),(Item=(Type=Bone,Name=\"upperarm_twist_01_r\"),Transform=(Rotation=(X=0,Y=0,Z=0,W=1),Translation=(X=0,Y=0,Z=0),Scale3D=(X=1,Y=1,Z=1)) ...(truncated)"
    Weight=1
    WeightMinimum=0
    ExecutePin=@RigUnit_BeginExecution.ExecutePin
   )
   > FN_DeformRig_FNC_1 = RigVMFunctionReferenceNode(
    DeformRig_ Curve="(deformRig_l_upperarm_up_alpha_72_#1=False,deformRig_r_upperarm_up_alpha_73_#2=False,deformRig_l_upperarm_fr_alpha_74_#3=False,deformRig_r_upperarm_fr_alpha_75_CB51B8 ...(truncated)"
    DeformRig_Bool="(upperarm_1_#4=True,elbow_3_#5=True,wrist_7_#6=True,groin_8_#7=True,glute_17_#8=True,knee_9_40C408324CA261 ...(truncated)"
    DeformRig_Debug=(upperarm_debug_18_#4=False,wrist_debug_20_#6=False,lat_debug_26_#9=False,groinGlute_debug_27_#7=False)
    DeformRig_Value="(upperarm_up_18_#10=(Rotation=(X=0,Y=0,Z=0,W=1),Translation=(X=0,Y=0,Z=4),Scale3D=(X=0.7,Y=1,Z=1)),upperarm_fwd_fr_19_#11=(Rotation=(X=0,Y=0,Z=0,W=1),Translation=(X=0,Y=4,Z=-2),Scale3D=(X=1,Y=1,Z=1) ...(truncated)"
   )
   > &RigUnit_ParentConstraint(
    AdvancedSettings=(InterpolationType=Average,RotationOrderForFilter=XZY)
    Child=(Type=Bone,Name="deform_foot_l_01")
    Filter=(TranslationFilter=(bX=true,bY=true,bZ=true),RotationFilter=(bX=True,bY=True,bZ=True),ScaleFilter=(bX=True,bY=True,bZ=True))
    Parents=((Item=(Type=Bone,Name="foot_l"),Weight=0.6),(Item=(Type=Bone,Name="calf_l"),Weight=0.4))
    Weight=1
    bMaintainOffset=true
   )
   > &RigUnit_ParentConstraint(
    AdvancedSettings=(InterpolationType=Average,RotationOrderForFilter=XZY)
    Child=(Type=Bone,Name="deform_foot_r_01")
    Filter=(TranslationFilter=(bX=true,bY=true,bZ=true),RotationFilter=(bX=True,bY=True,bZ=True),ScaleFilter=(bX=True,bY=True,bZ=True))
    Parents=((Item=(Type=Bone,Name="foot_r"),Weight=0.6),(Item=(Type=Bone,Name="calf_r"),Weight=0.4))
    Weight=1
    bMaintainOffset=true
   )
   > &RigUnit_ModifyTransforms(
    ItemToModify="((Item=(Type=Bone,Name=\"deform_foot_l_02\"),Transform=(Rotation=(X=0,Y=0,Z=0.0443,W=0.999),Translation=(X=0.0411,Y=-2.4476,Z=0),Scale3D=(X=1,Y=1,Z=1))),(Item=(Type=Bone,Name=\"deform_foot_l_01\"),Transform=(Rotation=(X=0,Y=0,Z=0,W=1),Translation=(X=0,Y=0,Z=0) ...(truncated)"
    Weight=1
    WeightMinimum=0
   )
   > &RigUnit_ParentConstraint(
    AdvancedSettings=(RotationOrderForFilter=XZY)
    Child=(Type=Bone,Name="base_Tshirt_fr_c")
    Parents="((Item=(Type=Bone,Name=\"spine_05\"),Weight=0.0084),(Item=(Type=Bone,Name=\"spine_01\"),Weight=0.0255),(Item=(Type=Bone,Name=\"pelvis\"),Weight=0.0671),(Item=(Type=Bone,Name=\"spine_04\"),Weight=0.1497),(Item=(Type=Bone,Name=\"spine_03\"),Weight=0.3738),(Item=(Type= ...(truncated)"
    Weight=1
   )
   > &RigUnit_ParentConstraint(
    AdvancedSettings=(RotationOrderForFilter=XZY)
    Child=(Type=Bone,Name="base_Tshirt_bk_c")
    Parents="((Item=(Type=Bone,Name=\"spine_05\"),Weight=0.0153),(Item=(Type=Bone,Name=\"clavicle_l\"),Weight=0.0204),(Item=(Type=Bone,Name=\"clavicle_r\"),Weight=0.0242),(Item=(Type=Bone,Name=\"spine_04\"),Weight=0.2081),(Item=(Type=Bone,Name=\"dyn_Tshirt_bk_c_01\"),Weight=0.27 ...(truncated)"
    Weight=1
   )
   > &RigUnit_ParentConstraint(
    AdvancedSettings=(RotationOrderForFilter=XZY)
    Child=(Type=Bone,Name="base_Tshirt_side_r")
    Parents="((Item=(Type=Bone,Name=\"spine_05\"),Weight=0.0091),(Item=(Type=Bone,Name=\"clavicle_r\"),Weight=0.0197),(Item=(Type=Bone,Name=\"spine_04\"),Weight=0.2278),(Item=(Type=Bone,Name=\"dyn_Tshirt_side_r_01\"),Weight=0.309),(Item=(Type=Bone,Name=\"spine_03\"),Weight=0.434 ...(truncated)"
    Weight=1
   )
   > &RigUnit_ParentConstraint(
    AdvancedSettings=(RotationOrderForFilter=XZY)
    Child=(Type=Bone,Name="base_Tshirt_side_l")
    Parents="((Item=(Type=Bone,Name=\"spine_05\"),Weight=0.0091),(Item=(Type=Bone,Name=\"clavicle_l\"),Weight=0.0197),(Item=(Type=Bone,Name=\"spine_04\"),Weight=0.2278),(Item=(Type=Bone,Name=\"dyn_Tshirt_side_l_01\"),Weight=0.3177),(Item=(Type=Bone,Name=\"spine_03\"),Weight=0.42 ...(truncated)"
    Weight=1
   )
   > Delta Rotation X = RigVMFunctionReferenceNode(Child=thigh_r, Parent=thigh_r_null, Source Maximum=-50, Source Minimum=0)
   Multiply = &RigVMFunction_MathDoubleMul(A=@"Delta Rotation X".Delta Rotation, B=RigVMVariableNode() {variable=IsIdle}.Value)
   > &RigUnit_ModifyTransforms(
    ItemToModify=((Item=(Type=Bone,Name="dyn_boot_zipper_r_01"),Transform=(Rotation=(X=0,Y=-0.6428,Z=0,W=0.766),Translation=(X=0,Y=0,Z=0),Scale3D=(X=1,Y=1,Z=1))))
    WeightMinimum=0
    Weight=@Multiply.Result
   )
   > Delta Rotation X_1 = RigVMFunctionReferenceNode(Child=thigh_l, Parent=thigh_l_null, Source Maximum=-50, Source Minimum=0)
   > &RigUnit_ModifyTransforms(
    ItemToModify=((Item=(Type=Bone,Name="dyn_boot_zipper_l_01"),Transform=(Rotation=(X=0,Y=-0.5,Z=0,W=0.866),Translation=(X=0,Y=0,Z=0),Scale3D=(X=1,Y=1,Z=1))))
    WeightMinimum=0
    Weight=&RigVMFunction_MathDoubleMul(A=@"Delta Rotation X_1".Delta Rotation, B=RigVMVariableNode() {variable=IsIdle}.Value).Result
   )
   > n1 = &RigUnit_SphericalPoseReader(
    ActiveRegionScaleFactors=(PositiveWidth=1,NegativeWidth=0.1,PositiveHeight=1,NegativeHeight=1)
    ActiveRegionSize=0.3
    Debug=(bDrawDebug=False,bDraw2D=false,bDrawLocalAxes=false,DebugScale=30,DebugSegments=30,DebugThickness=0.2)
    DriverAxis=(X=-1,Y=0,Z=0)
    DriverItem=(Type=Bone,Name="thigh_l")
    FalloffRegionScaleFactors=(PositiveWidth=1,NegativeWidth=0.5,PositiveHeight=1,NegativeHeight=1)
    FalloffSize=0.5
    OptionalParentItem=(Type=Bone,Name="pelvis")
    RotationOffset=(X=90,Y=0,Z=0)
   )
   n2 = RigVMVariableNode() {variable=Moving2D}
   n3 = &RigVMFunction_MathBoolNot(Value=@n2.Value)
   n4 = &RigVMFunction_AlphaInterp(
    InterpSpeedDecreasing=5
    InterpSpeedIncreasing=5
    Scale=1
    bClampResult=False
    Value=&RigVMFunction_MathDoubleMul(B=&RigVMFunction_MathBoolToFloat(Value=@n3.Result).Result, A=@n1.OutputParam).Result
   )
   > &RigUnit_ModifyTransforms(
    ItemToModify="((Item=(Type=Bone,Name=\"deform_glute_l\"),Transform=(Rotation=(X=0,Y=0,Z=0,W=1),Translation=(X=-3,Y=0,Z=5),Scale3D=(X=1,Y=1,Z=1))),(Item=(Type=Bone,Name=\"deform_glute_r\"),Transform=(Rotation=(X=0,Y=0,Z=0,W=1),Translation=(X=-1,Y=0,Z=0),Scale3D=(X=1,Y=1,Z=1) ...(truncated)"
    Weight=@n4.Result
   )
   FN_DeformRig_FNC_2 = RigVMFunctionReferenceNode(
    DeformRig_ Curve="(deformRig_l_upperarm_up_alpha_72_#1=False,deformRig_r_upperarm_up_alpha_73_#2=False,deformRig_l_upperarm_fr_alpha_74_#3=False,deformRig_r_upperarm_fr_alpha_75_CB51B8 ...(truncated)"
    DeformRig_Bool="(upperarm_1_#4=True,elbow_3_#5=True,wrist_7_#6=True,groin_8_#7=True,glute_17_#8=True,knee_9_40C408324CA261 ...(truncated)"
    DeformRig_Debug=(upperarm_debug_18_#4=False,wrist_debug_20_#6=False,lat_debug_26_#9=False,groinGlute_debug_27_#7=False)
    DeformRig_Value="(upperarm_up_18_#10=(Rotation=(X=0,Y=0,Z=0,W=1),Translation=(X=0,Y=0,Z=4),Scale3D=(X=0.7,Y=1,Z=1)),upperarm_fwd_fr_19_#11=(Rotation=(X=0,Y=0,Z=0,W=1),Translation=(X=0,Y=-8,Z=0),Scale3D=(X=1,Y=1,Z=1) ...(truncated)"
    ExecuteContext=()
   )
   RigVMFunction_MathBoolOr = &RigVMFunction_MathBoolOr(A=@n2.Value, B=@n2.Value)
  ]}
]
hierarchy [
 root {loc=<0 0 0>, rot=<0 0 0 1>}
 pelvis {parent=root, loc=<0 -0.5301 98.0571>, rot=<0.0127 0.707 -0.0127 -0.707>}
 spine_01 {parent=pelvis, loc=<6.1208 -0.9328 0>, rot=<0 0 0.034 -0.9994>}
 spine_02 {parent=spine_01, loc=<6.5997 -0.2095 0>, rot=<0 0 0.0238 -0.9997>}
 spine_03 {parent=spine_02, loc=<6.4914 -0.1237 0>, rot=<0 0 0.0149 -0.9999>}
 spine_04 {parent=spine_03, loc=<6.9598 0.0794 0>, rot=<0 0 -0.0132 -0.9999>}
 spine_05 {parent=spine_04, loc=<6.9987 0 0>, rot=<0 0 -0.0015 -1>}
 neck_01 {parent=spine_05, loc=<11.3138 0.1529 0>, rot=<0 0 -0.2003 -0.9797>}
 neck_02 {parent=neck_01, loc=<4.5627 0 0>, rot=<0 0 0.0504 -0.9987>}
 head {parent=neck_02, loc=<5.3914 0 0>, rot=<0 0 0.115 -0.9934>}
 clavicle_l {parent=spine_05, loc=<5.4831 -0.1053 -3.2946>, rot=<0.0465 -0.7289 -0.009 -0.6829>}
 upperarm_l {parent=clavicle_l, loc=<9.8658 -1.2136 -0.4105>, rot=<0.0164 -0.3245 -0.0269 -0.9454>}
 lowerarm_l {parent=upperarm_l, loc=<26.465 -0.8345 0.259>, rot=<0.0036 0.0161 -0.2298 -0.9731>}
 hand_l {parent=lowerarm_l, loc=<22.7355 -0.3203 -0.6105>, rot=<-0.6784 -0.0223 0.0735 -0.7307>}
 thumb_01_l {parent=hand_l, loc=<1.3118 -0.3191 -2.459>, rot=<0.6546 -0.4042 0.0664 -0.6354>}
 thumb_02_l {parent=thumb_01_l, loc=<5.9336 -0.0448 0.0401>, rot=<-0.0037 0.0213 0.298 -0.9543>}
 thumb_03_l {parent=thumb_02_l, loc=<2.2223 0.3682 -0.0907>, rot=<0.0118 -0.013 0.0066 -0.9998>}
 index_metacarpal_l {parent=hand_l, loc=<1.6535 0.0183 -2.96>, rot=<0.0158 -0.0124 -0.0173 -0.9996>}
 index_01_l {parent=index_metacarpal_l, loc=<8.3008 0.0357 -0.0524>, rot=<-0.0044 0.0485 0.0924 -0.9945>}
 index_02_l {parent=index_01_l, loc=<3.3587 -0.3004 0.0452>, rot=<-0.002 0.0027 0.3225 -0.9466>}
 index_03_l {parent=index_02_l, loc=<2.5578 -0.123 -0.0252>, rot=<-0.002 0.0067 0.2558 -0.9667>}
 middle_metacarpal_l {parent=hand_l, loc=<2.1856 -0.1725 -0.5984>, rot=<0.006 0.0168 -0.0329 -0.9993>}
 middle_01_l {parent=middle_metacarpal_l, loc=<7.3275 -0.2447 0.0129>, rot=<-0.0208 0.0042 0.0858 -0.9961>}
 middle_02_l {parent=middle_01_l, loc=<4.0123 0.1401 0.1175>, rot=<-0.0218 0.0258 0.3512 -0.9357>}
 middle_03_l {parent=middle_02_l, loc=<2.6077 0.0534 0.1373>, rot=<0.0082 0.0188 0.1708 -0.9851>}
 ring_metacarpal_l {parent=hand_l, loc=<2.2074 -0.3625 1.3638>, rot=<-0.0284 0.0722 -0.0458 -0.9959>}
 ring_01_l {parent=ring_metacarpal_l, loc=<7.0295 -0.6197 -0.1446>, rot=<-0.0489 -0.0252 0.1892 -0.9804>}
 ring_02_l {parent=ring_01_l, loc=<3.6174 0.4647 -0.1027>, rot=<0 0 0.2806 -0.9598>}
 ring_03_l {parent=ring_02_l, loc=<2.4152 0.114 0.0199>, rot=<0 0 0.1189 -0.9929>}
 pinky_metacarpal_l {parent=hand_l, loc=<1.7616 -0.9753 3.3814>, rot=<-0.0397 0.0814 -0.069 -0.9935>}
 pinky_01_l {parent=pinky_metacarpal_l, loc=<7.1709 -0.4831 -0.1094>, rot=<-0.0654 -0.0035 0.2242 -0.9723>}
 pinky_02_l {parent=pinky_01_l, loc=<2.8274 0.0893 -0.1043>, rot=<0.0037 0.0011 0.2854 -0.9584>}
 pinky_03_l {parent=pinky_02_l, loc=<2.2046 0.0307 -0.0078>, rot=<-0.0002 0.0098 0.1445 -0.9895>}
 deform_wrist_dn_base_l {parent=hand_l, loc=<0 -1.5326 0>, rot=<0 0 0 -1>}
 deform_wrist_dn_l {parent=deform_wrist_dn_base_l, loc=<0 0 0>, rot=<0 0 0 1>}
 deform_wrist_up_base_l {parent=hand_l, loc=<0 2.3009 0>, rot=<0 0 0 -1>}
 deform_wrist_up_l {parent=deform_wrist_up_base_l, loc=<0 0 0>, rot=<0 0 0 1>}
 lowerarm_twist_02_l {parent=lowerarm_l, loc=<10.6707 0 0>, rot=<0 0 0 1>}
 lowerarm_twist_01_l {parent=lowerarm_l, loc=<18.316 0 0>, rot=<0 0 0 1>}
 deform_elbow_in_base_l {parent=lowerarm_l, loc=<0.1834 3.6062 -0.2572>, rot=<0 0 0 -1>}
 deform_elbow_in_l {parent=deform_elbow_in_base_l, loc=<0 0 0>, rot=<0 0 0 1>}
 deform_elbow_out_base_l {parent=lowerarm_l, loc=<0 -4 0>, rot=<0 0 0 1>}
 deform_elbow_out_l {parent=deform_elbow_out_base_l, loc=<0 0 0>, rot=<0 0 0 1>}
 upperarm_twist_01_l {parent=upperarm_l, loc=<7.2135 -0.1134 0>, rot=<0 0 0 1>}
 deform_upperarm_bk_base_l {parent=upperarm_twist_01_l, loc=<-6.4053 -4.3746 0.314>, rot=<0 0 0 -1>}
 deform_upperarm_bk_l {parent=deform_upperarm_bk_base_l, loc=<0 0 0>, rot=<0 0 0 1>}
 deform_upperarm_fr_base_l {parent=upperarm_twist_01_l, loc=<-5.6312 5.092 0.3354>, rot=<0 0 0 -1>}
 deform_upperarm_fr_l {parent=deform_upperarm_fr_base_l, loc=<0 0 0>, rot=<0 0 0 1>}
 deform_upperarm_up_base_l {parent=upperarm_twist_01_l, loc=<-8.891 0.4825 3.9517>, rot=<0 0 0 -1>}
 deform_upperarm_up_l {parent=deform_upperarm_up_base_l, loc=<0 0 0>, rot=<0 0 0 1>}
 base_sleeve_top_l {parent=upperarm_twist_01_l, loc=<-0.4964 -0.5803 4.133>, rot=<-0.0074 0.078 0.0019 -0.9969>}
 dyn_sleeve_top_l_01 {parent=base_sleeve_top_l, loc=<0 0 0>, rot=<0 0 0 1>}
 base_sleeve_front_l {parent=upperarm_twist_01_l, loc=<0.265 4.4294 0.5014>, rot=<0.571 0.0729 -0.0162 -0.8175>}
 dyn_sleeve_front_l_01 {parent=base_sleeve_front_l, loc=<0 0 0>, rot=<0 0 0 1>}
 base_sleeve_under_l {parent=upperarm_twist_01_l, loc=<0.9962 1.2272 -5.0792>, rot=<0.998 -0.002 -0.0559 -0.0282>}
 dyn_sleeve_under_l_01 {parent=base_sleeve_under_l, loc=<0 0 0>, rot=<0 0 0 1>}
 base_sleeve_back_l {parent=upperarm_twist_01_l, loc=<0.5526 -4.8807 -1.7629>, rot=<-0.7361 0.1145 0.0248 -0.6667>}
 dyn_sleeve_back_l_01 {parent=base_sleeve_back_l, loc=<0 0 0>, rot=<0 0 0 1>}
 upperarm_twist_02_l {parent=upperarm_l, loc=<19.7998 -0.1134 0>, rot=<0 0 0 1>}
 clavicle_r {parent=spine_05, loc=<5.4831 -0.1053 3.2946>, rot=<0.7289 0.0465 -0.6829 0.009>}
 upperarm_r {parent=clavicle_r, loc=<-9.8658 1.2136 0.4105>, rot=<0.0164 -0.3245 -0.0269 -0.9454>}
 lowerarm_r {parent=upperarm_r, loc=<-26.465 0.8345 -0.259>, rot=<0.0036 0.0161 -0.2298 -0.9731>}
 hand_r {parent=lowerarm_r, loc=<-22.7355 0.3203 0.6106>, rot=<-0.6784 -0.0223 0.0735 -0.7307>}
 thumb_01_r {parent=hand_r, loc=<-1.3118 0.3191 2.459>, rot=<0.6546 -0.4042 0.0664 -0.6354>}
 thumb_02_r {parent=thumb_01_r, loc=<-5.9336 0.0448 -0.0401>, rot=<-0.0037 0.0213 0.298 -0.9543>}
 thumb_03_r {parent=thumb_02_r, loc=<-2.2223 -0.3682 0.0907>, rot=<0.0118 -0.013 0.0066 -0.9998>}
 index_metacarpal_r {parent=hand_r, loc=<-1.6535 -0.0183 2.96>, rot=<0.0158 -0.0124 -0.0173 -0.9996>}
 index_01_r {parent=index_metacarpal_r, loc=<-8.3008 -0.0357 0.0524>, rot=<-0.0044 0.0485 0.0924 -0.9945>}
 index_02_r {parent=index_01_r, loc=<-3.3587 0.3004 -0.0452>, rot=<-0.002 0.0027 0.3225 -0.9466>}
 index_03_r {parent=index_02_r, loc=<-2.5578 0.123 0.0252>, rot=<-0.002 0.0067 0.2558 -0.9667>}
 middle_metacarpal_r {parent=hand_r, loc=<-2.1856 0.1725 0.5984>, rot=<0.006 0.0168 -0.0329 -0.9993>}
 middle_01_r {parent=middle_metacarpal_r, loc=<-7.3275 0.2447 -0.0129>, rot=<-0.0208 0.0042 0.0858 -0.9961>}
 middle_02_r {parent=middle_01_r, loc=<-4.0123 -0.1401 -0.1175>, rot=<-0.0218 0.0258 0.3512 -0.9357>}
 middle_03_r {parent=middle_02_r, loc=<-2.6077 -0.0534 -0.1373>, rot=<0.0082 0.0188 0.1708 -0.9851>}
 ring_metacarpal_r {parent=hand_r, loc=<-2.2074 0.3625 -1.3638>, rot=<-0.0284 0.0722 -0.0458 -0.9959>}
 ring_01_r {parent=ring_metacarpal_r, loc=<-7.0295 0.6197 0.1446>, rot=<-0.0489 -0.0252 0.1892 -0.9804>}
 ring_02_r {parent=ring_01_r, loc=<-3.6174 -0.4647 0.1027>, rot=<0 0 0.2806 -0.9598>}
 ring_03_r {parent=ring_02_r, loc=<-2.4152 -0.114 -0.0199>, rot=<0 0 0.1189 -0.9929>}
 pinky_metacarpal_r {parent=hand_r, loc=<-1.7616 0.9753 -3.3814>, rot=<-0.0397 0.0814 -0.069 -0.9935>}
 pinky_01_r {parent=pinky_metacarpal_r, loc=<-7.1709 0.4831 0.1094>, rot=<-0.0654 -0.0035 0.2242 -0.9723>}
 pinky_02_r {parent=pinky_01_r, loc=<-2.8274 -0.0893 0.1043>, rot=<0.0037 0.0011 0.2854 -0.9584>}
 pinky_03_r {parent=pinky_02_r, loc=<-2.2046 -0.0307 0.0078>, rot=<-0.0002 0.0098 0.1445 -0.9895>}
 deform_wrist_dn_base_r {parent=hand_r, loc=<-0.0001 1.5328 0>, rot=<0 0 0 -1>}
 deform_wrist_dn_r {parent=deform_wrist_dn_base_r, loc=<0 0 0>, rot=<0 0 0 1>}
 deform_wrist_up_base_r {parent=hand_r, loc=<-0.0002 -2.3007 -0.0001>, rot=<0 0 0 -1>}
 deform_wrist_up_r {parent=deform_wrist_up_base_r, loc=<0 0 0>, rot=<0 0 0 1>}
 lowerarm_twist_02_r {parent=lowerarm_r, loc=<-10.6707 0 0>, rot=<0 0 0 -1>}
 lowerarm_twist_01_r {parent=lowerarm_r, loc=<-18.316 0 0>, rot=<0 0 0 -1>}
 deform_elbow_in_base_r {parent=lowerarm_r, loc=<-0.1834 -3.6062 0.2572>, rot=<0 0 0 -1>}
 deform_elbow_in_r {parent=deform_elbow_in_base_r, loc=<0 0 0>, rot=<0 0 0 1>}
 deform_elbow_out_base_r {parent=lowerarm_r, loc=<0 4 -0.0001>, rot=<0 0 0 -1>}
 deform_elbow_out_r {parent=deform_elbow_out_base_r, loc=<0 0 0>, rot=<0 0 0 1>}
 upperarm_twist_01_r {parent=upperarm_r, loc=<-7.2135 0.1134 0>, rot=<0 0 0 -1>}
 deform_upperarm_bk_base_r {parent=upperarm_twist_01_r, loc=<6.4053 4.3746 -0.314>, rot=<0 0 0 -1>}
 deform_upperarm_bk_r {parent=deform_upperarm_bk_base_r, loc=<0 0 0>, rot=<0 0 0 1>}
 deform_upperarm_fr_base_r {parent=upperarm_twist_01_r, loc=<5.6312 -5.092 -0.3354>, rot=<0 0 0 -1>}
 deform_upperarm_fr_r {parent=deform_upperarm_fr_base_r, loc=<0 0 0>, rot=<0 0 0 1>}
 deform_upperarm_up_base_r {parent=upperarm_twist_01_r, loc=<8.891 -0.4825 -3.9517>, rot=<0 0 0 -1>}
 deform_upperarm_up_r {parent=deform_upperarm_up_base_r, loc=<0 0 0>, rot=<0 0 0 1>}
 base_sleeve_front_r {parent=upperarm_twist_01_r, loc=<-0.2648 -4.4294 -0.5016>, rot=<0.0162 -0.8175 0.571 -0.0729>}
 dyn_sleeve_front_r_01 {parent=base_sleeve_front_r, loc=<0 0 0>, rot=<0 0 0 1>}
 base_sleeve_under_r {parent=upperarm_twist_01_r, loc=<-0.9964 -1.2272 5.0793>, rot=<0.0559 -0.0282 0.998 0.002>}
 dyn_sleeve_under_r_01 {parent=base_sleeve_under_r, loc=<0 0 0>, rot=<0 0 0 1>}
 base_sleeve_back_r {parent=upperarm_twist_01_r, loc=<-0.5527 4.8806 1.763>, rot=<0.0248 0.6667 0.7361 0.1145>}
 dyn_sleeve_back_r_01 {parent=base_sleeve_back_r, loc=<0 0 0>, rot=<0 0 0 1>}
 base_sleeve_top_r {parent=upperarm_twist_01_r, loc=<0.4962 0.5803 -4.1329>, rot=<-0.0019 -0.9969 -0.0074 -0.078>}
 dyn_sleeve_top_r_01 {parent=base_sleeve_top_r, loc=<0 0 0>, rot=<0 0 0 1>}
 upperarm_twist_02_r {parent=upperarm_r, loc=<-19.7998 0.1134 0>, rot=<0 0 0 -1>}
 attach_backpack {parent=spine_05, loc=<-0.123 -13.473 0>, rot=<-0.0093 -0.707 0.0093 -0.707>}
 attach_cape {parent=spine_05, loc=<9.017 -11.714 0>, rot=<0.0616 -0.7044 -0.0616 -0.7044>}
 deform_pec_aim_base_l {parent=spine_04, loc=<3.2456 9.3309 -1.2237>, rot=<0.0603 -0.4033 0.3318 -0.8507>}
 deform_pec_aim_l {parent=deform_pec_aim_base_l, loc=<0 0 0>, rot=<0 0 0 1>}
 deform_pec_base_l {parent=deform_pec_aim_l, loc=<9.3665 0 0>, rot=<0 0 0 1>}
 deform_pec_l {parent=deform_pec_base_l, loc=<0 0 0>, rot=<0 0 0 1>}
 deform_pec_aim_base_r {parent=spine_04, loc=<3.2452 9.3309 1.2237>, rot=<-0.4033 -0.0603 0.8507 0.3318>}
 deform_pec_aim_r {parent=deform_pec_aim_base_r, loc=<0 0 0>, rot=<0 0 0 1>}
 deform_pec_base_r {parent=deform_pec_aim_r, loc=<-9.3667 -0.0003 -0.0004>, rot=<0 0 0 1>}
 deform_pec_r {parent=deform_pec_base_r, loc=<0 0 0>, rot=<0 0 0 1>}
 base_Tshirt_side_zero {parent=spine_02, loc=<5.8509 1.8897 -0.0005>, rot=<0.0399 -0.9988 0.0011 0.0279>}
 base_Tshirt_fr_c {parent=base_Tshirt_side_zero, loc=<-0.1961 8.7129 -0.0133>, rot=<0.7068 0.0198 0.0198 -0.7068>}
 dyn_Tshirt_fr_c_01 {parent=base_Tshirt_fr_c, loc=<0 0 0>, rot=<0 0 0 1>}
 base_Tshirt_bk_c {parent=base_Tshirt_side_zero, loc=<-0.158 -7.0898 -0.0049>, rot=<-0.7069 0.0213 -0.0183 -0.7068>}
 dyn_Tshirt_bk_c_01 {parent=base_Tshirt_bk_c, loc=<0 0 0>, rot=<0 0 0 1>}
 base_Tshirt_side_l {parent=base_Tshirt_side_zero, loc=<-0.7911 0 10.3458>, rot=<0 0.0382 0 -0.9993>}
 dyn_Tshirt_side_l_01 {parent=base_Tshirt_side_l, loc=<0 0 0>, rot=<0 -0.0382 0 -0.9993>}
 base_Tshirt_side_r {parent=base_Tshirt_side_zero, loc=<0.3692 0 -10.3705>, rot=<0.9998 0 0.0178 0>}
 dyn_Tshirt_side_r_01 {parent=base_Tshirt_side_r, loc=<0 0 0>, rot=<0 0 0 1>}
 thigh_l {parent=pelvis, loc=<-1.0855 -0.2109 -8.7053>, rot=<0.0009 0.0355 0.0155 -0.9993>}
 calf_l {parent=thigh_l, loc=<-42.5284 -0.0019 -0.7297>, rot=<0.0006 -0.0161 -0.0391 -0.9991>}
 foot_l {parent=calf_l, loc=<-46.7401 -2.5014 -0.592>, rot=<0.0427 -0.0238 0.0353 -0.9982>}
 ball_l {parent=foot_l, loc=<-3.5658 10.8066 -0.1544>, rot=<0.0001 -0.0001 -0.7063 -0.7079>}
 deform_foot_l_01 {parent=foot_l, loc=<0 0 0>, rot=<0 0 0 1>}
 deform_foot_l_02 {parent=deform_foot_l_01, loc=<0 -4.6061 0>, rot=<0 0 0 1>}
 calf_twist_02_l {parent=calf_l, loc=<-19.1546 -1.7943 -0.2638>, rot=<0.0237 0.0095 -0.0035 -0.9997>}
 base_boot_strap_l {parent=calf_twist_02_l, loc=<-6.0878 -7.0352 -1.4769>, rot=<0.6592 -0.2146 0.2019 -0.6919>}
 dyn_boot_strap_l_01 {parent=base_boot_strap_l, loc=<0 0 0>, rot=<0 0 0 1>}
 base_boot_zipper_l {parent=calf_twist_02_l, loc=<-7.5202 -2.497 4.3481>, rot=<-0.0701 0.0377 0.9957 0.0484>}
 dyn_boot_zipper_l_01 {parent=base_boot_zipper_l, loc=<0 0 0>, rot=<0 0 0 1>}
 calf_twist_01_l {parent=calf_l, loc=<-35.8483 -1.9197 -0.5794>, rot=<0.0237 0.0095 -0.0035 -0.9997>}
 deform_knee_in_base_l {parent=calf_l, loc=<0.8031 -5.0231 0.2456>, rot=<0 0 0 -1>}
 deform_knee_in_l {parent=deform_knee_in_base_l, loc=<0 0 0>, rot=<0 0 0 1>}
 deform_knee_out_base_l {parent=calf_l, loc=<0 4 0>, rot=<0 0 0 -1>}
 deform_knee_out_l {parent=deform_knee_out_base_l, loc=<0 0 0>, rot=<0 0 0 1>}
 thigh_twist_01_l {parent=thigh_l, loc=<-27.952 0 0>, rot=<0.0013 0 0 -1>}
 deform_groin_base_l {parent=thigh_l, loc=<0.4801 5.6776 -0.2486>, rot=<0 0 0 -1>}
 deform_groin_l {parent=deform_groin_base_l, loc=<0 0 0>, rot=<0 0 0 1>}
 thigh_r {parent=pelvis, loc=<-1.0855 -0.2109 8.7053>, rot=<0.0355 -0.0009 0.9993 0.0155>}
 calf_r {parent=thigh_r, loc=<42.5284 0.0019 0.7297>, rot=<0.0006 -0.0161 -0.0391 -0.9991>}
 foot_r {parent=calf_r, loc=<46.7401 2.5014 0.592>, rot=<0.0427 -0.0238 0.0353 -0.9982>}
 ball_r {parent=foot_r, loc=<3.5658 -10.8066 0.1544>, rot=<0.0001 -0.0001 -0.7063 -0.7079>}
 deform_foot_r_01 {parent=foot_r, loc=<0 0 0>, rot=<1 0 0 0>}
 deform_foot_r_02 {parent=deform_foot_r_01, loc=<0 -4.6061 0>, rot=<0 0 0 -1>}
 calf_twist_02_r {parent=calf_r, loc=<19.1546 1.7943 0.2638>, rot=<0.0237 0.0095 -0.0035 -0.9997>}
 base_boot_strap_r {parent=calf_twist_02_r, loc=<6.3268 7.0924 1.4708>, rot=<-0.1568 -0.6961 0.6696 0.2059>}
 dyn_boot_strap_r_01 {parent=base_boot_strap_r, loc=<0 0 0>, rot=<0 0 0 1>}
 base_boot_zipper_r {parent=calf_twist_02_r, loc=<7.5143 2.5044 -4.3298>, rot=<0.9955 -0.0628 0.0622 0.0355>}
 dyn_boot_zipper_r_01 {parent=base_boot_zipper_r, loc=<0 0 0>, rot=<0 0 0 1>}
 calf_twist_01_r {parent=calf_r, loc=<35.8483 1.9197 0.5794>, rot=<0.0237 0.0095 -0.0035 -0.9997>}
 deform_knee_in_base_r {parent=calf_r, loc=<-0.8031 5.0231 -0.2457>, rot=<0 0 0 -1>}
 deform_knee_in_r {parent=deform_knee_in_base_r, loc=<0 0 0>, rot=<0 0 0 1>}
 deform_knee_out_base_r {parent=calf_r, loc=<0 -4 0>, rot=<0 0 0 -1>}
 deform_knee_out_r {parent=deform_knee_out_base_r, loc=<0 0 0>, rot=<0 0 0 1>}
 thigh_twist_01_r {parent=thigh_r, loc=<27.952 0 0>, rot=<0.0013 0 0 -1>}
 deform_groin_base_r {parent=thigh_r, loc=<-0.4801 -5.6776 0.2486>, rot=<0 0 0 -1>}
 deform_groin_r {parent=deform_groin_base_r, loc=<0 0 0>, rot=<0 0 0 1>}
 deform_glute_base_l {parent=pelvis, loc=<-7.3466 -9.0213 -9.1245>, rot=<0.0172 0.031 -0.4477 -0.8935>}
 deform_glute_l {parent=deform_glute_base_l, loc=<0 0 0>, rot=<0 0 0 1>}
 deform_glute_base_r {parent=pelvis, loc=<-7.3464 -9.0213 9.1245>, rot=<0.031 -0.0172 0.8935 -0.4477>}
 deform_glute_r {parent=deform_glute_base_r, loc=<0 0 0>, rot=<0 0 0 1>}
 thigh_r_null {parent=pelvis, type=Null, loc=<-1.0855 -0.2109 8.7053>, rot=<0.0355 -0.0009 0.9993 0.0155>}
 thigh_l_null {parent=pelvis, type=Null, loc=<-1.0855 -0.2109 -8.7053>, rot=<0.0009 0.0355 0.0155 -0.9993>}
]
variables [
 IsIdle {type=double, default=0}
 Moving2D {type=bool, default=False}
]
