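# Post-copy animation Blueprint for the TriadSpear female body cosmetic, adding secondary motion to clothing.
# It copies the attached parent's pose through a Control Rig fed idle and movement flags.
# Chained spring-driven AnimDynamics nodes then swing sleeves, boot straps, zippers and T-shirt bones.
# Each update derives isIdle from movement state and stores gravity and rigid-body settings from a parameter asset, though the dynamics nodes show fixed gravity.
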AnimBlueprint /BRCosmetics/Characters/Player/Female/Medium/Bodies/F_MED_TriadSpear/Meshes/F_MED_TriadSpear_AnimBP
target_skeleton=&/F_MED_TriadSpear_Skeleton
graphs [
 AnimGraph {nodes [
   n1 = Control Rig!(
    Source=Copy Pose From Mesh() {bUseAttachedParent=true}
    Alpha
    IsIdle:double=Get isIdle() -> (isIdle:double)
    Moving2D:bool=Get bIsMoving2D()
   ) -> (Pose) {controlRigAssetReference {blueprintRigClass=&/F_MED_TriadSpear_CtrlRig.F_MED_TriadSpear_CtrlRig_C}}
   n2 = Anim Dynamics - Bone: dyn_sleeve_top_l_01!(ComponentPose=Local To Component(LocalPose=@n1), Alpha=1, GravityOverride:Vector="0,0,-980", ExternalForce="0,0,0") -> (Pose) {
    angularDampingOverride=1
    angularSpringConstant=8000
    bAngularSpring=true
    bOverrideAngularDamping=true
    bUseGravityOverride=true
    gravityScale=0.35
    simulationSpace=RootRelative
    alphaScaleBiasClamp {scale=0.5}
    boundBone {boneName=dyn_sleeve_top_l_01}
    componentAppliedLinearAccClamp=<2000 2000 2000>
    componentLinearAccScale=<0.5 0.5 0.5>
    componentLinearVelScale=<0.5 0.5 0.5>
    gravityOverride {z=-980}
    physicsBodyDefinitions [{
      boundBone {boneName=dyn_sleeve_top_l_01}
      boxExtents=<20 30 30>
      constraintSetup {angularLimitsMax=<0 10 5>, angularLimitsMin=<0 -30 -5>, angularTarget=<1 0 0>}
      localJointOffset=<4 0 0>
     }]
    simSpaceSettings {maxAngularAcceleration=40, maxAngularVelocity=8, simSpaceAngularAlpha=0.5}
   }
   n3 = Anim Dynamics - Bone: dyn_sleeve_front_l_01!(ComponentPose=@n2, Alpha=1, GravityOverride:Vector="0,0,-980", ExternalForce="0,0,0") -> (Pose) {
    angularDampingOverride=1
    angularSpringConstant=8000
    bAngularSpring=true
    bOverrideAngularDamping=true
    bUseGravityOverride=true
    gravityScale=0.35
    simulationSpace=RootRelative
    alphaScaleBiasClamp {scale=0.5}
    boundBone {boneName=dyn_sleeve_front_l_01}
    componentAppliedLinearAccClamp=<2000 2000 2000>
    componentLinearAccScale=<0.5 0.5 0.5>
    componentLinearVelScale=<0.5 0.5 0.5>
    gravityOverride {z=-980}
    physicsBodyDefinitions [{
      boundBone {boneName=dyn_sleeve_front_l_01}
      boxExtents=<20 30 30>
      constraintSetup {angularLimitsMax=<0 10 5>, angularLimitsMin=<0 -30 -5>, angularTarget=<1 0 0>}
      localJointOffset=<4 0 0>
     }]
    simSpaceSettings {maxAngularAcceleration=40, maxAngularVelocity=8, simSpaceAngularAlpha=0.5}
   }
   n4 = Anim Dynamics - Bone: dyn_sleeve_under_l_01!(ComponentPose=@n3, Alpha=1, GravityOverride:Vector="0,0,-980", ExternalForce="0,0,0") -> (Pose) {
    angularDampingOverride=1
    angularSpringConstant=8000
    bAngularSpring=true
    bOverrideAngularDamping=true
    bUseGravityOverride=true
    gravityScale=0.35
    simulationSpace=RootRelative
    alphaScaleBiasClamp {scale=0.5}
    boundBone {boneName=dyn_sleeve_under_l_01}
    componentAppliedLinearAccClamp=<2000 2000 2000>
    componentLinearAccScale=<0.5 0.5 0.5>
    componentLinearVelScale=<0.5 0.5 0.5>
    gravityOverride {z=-980}
    physicsBodyDefinitions [{
      boundBone {boneName=dyn_sleeve_under_l_01}
      boxExtents=<20 30 30>
      constraintSetup {angularLimitsMax=<0 10 5>, angularLimitsMin=<0 -30 -5>, angularTarget=<1 0 0>}
      localJointOffset=<4 0 0>
     }]
    simSpaceSettings {maxAngularAcceleration=40, maxAngularVelocity=8, simSpaceAngularAlpha=0.5}
   }
   n5 = Anim Dynamics - Bone: dyn_sleeve_back_l_01!(ComponentPose=@n4, Alpha=1, GravityOverride:Vector="0,0,-980", ExternalForce="0,0,0") -> (Pose) {
    angularDampingOverride=1
    angularSpringConstant=8000
    bAngularSpring=true
    bOverrideAngularDamping=true
    bUseGravityOverride=true
    gravityScale=0.35
    simulationSpace=RootRelative
    alphaScaleBiasClamp {scale=0.5}
    boundBone {boneName=dyn_sleeve_back_l_01}
    componentAppliedLinearAccClamp=<2000 2000 2000>
    componentLinearAccScale=<0.5 0.5 0.5>
    componentLinearVelScale=<0.5 0.5 0.5>
    gravityOverride {z=-980}
    physicsBodyDefinitions [{
      boundBone {boneName=dyn_sleeve_back_l_01}
      boxExtents=<20 30 30>
      constraintSetup {angularLimitsMax=<0 10 5>, angularLimitsMin=<0 -30 -5>, angularTarget=<1 0 0>}
      localJointOffset=<4 0 0>
     }]
    simSpaceSettings {maxAngularAcceleration=40, maxAngularVelocity=8, simSpaceAngularAlpha=0.5}
   }
   n6 = Anim Dynamics - Bone: dyn_sleeve_top_r_01!(ComponentPose=@n5, Alpha=1, GravityOverride:Vector="0,0,-980", ExternalForce="0,0,0") -> (Pose) {
    angularDampingOverride=1
    angularSpringConstant=8000
    bAngularSpring=true
    bOverrideAngularDamping=true
    bUseGravityOverride=true
    gravityScale=0.35
    simulationSpace=RootRelative
    alphaScaleBiasClamp {scale=0.5}
    boundBone {boneName=dyn_sleeve_top_r_01}
    componentAppliedLinearAccClamp=<2000 2000 2000>
    componentLinearAccScale=<0.5 0.5 0.5>
    componentLinearVelScale=<0.5 0.5 0.5>
    gravityOverride {z=-980}
    physicsBodyDefinitions [{
      boundBone {boneName=dyn_sleeve_top_r_01}
      boxExtents=<20 30 30>
      constraintSetup {angularLimitsMax=<0 10 5>, angularLimitsMin=<0 -30 -5>, angularTarget=<1 0 0>}
      localJointOffset=<4 0 0>
     }]
    simSpaceSettings {maxAngularAcceleration=40, maxAngularVelocity=8, simSpaceAngularAlpha=0.5}
   }
   n7 = Anim Dynamics - Bone: dyn_sleeve_front_r_01!(ComponentPose=@n6, Alpha=1, GravityOverride:Vector="0,0,-980", ExternalForce="0,0,0") -> (Pose) {
    angularDampingOverride=1
    angularSpringConstant=8000
    bAngularSpring=true
    bOverrideAngularDamping=true
    bUseGravityOverride=true
    gravityScale=0.35
    simulationSpace=RootRelative
    alphaScaleBiasClamp {scale=0.5}
    boundBone {boneName=dyn_sleeve_front_r_01}
    componentAppliedLinearAccClamp=<2000 2000 2000>
    componentLinearAccScale=<0.5 0.5 0.5>
    componentLinearVelScale=<0.5 0.5 0.5>
    gravityOverride {z=-980}
    physicsBodyDefinitions [{
      boundBone {boneName=dyn_sleeve_front_r_01}
      boxExtents=<20 30 30>
      constraintSetup {angularLimitsMax=<0 10 5>, angularLimitsMin=<0 -30 -5>, angularTarget=<1 0 0>}
      localJointOffset=<4 0 0>
     }]
    simSpaceSettings {maxAngularAcceleration=40, maxAngularVelocity=8, simSpaceAngularAlpha=0.5}
   }
   n8 = Anim Dynamics - Bone: dyn_sleeve_under_r_01!(ComponentPose=@n7, Alpha=1, GravityOverride:Vector="0,0,-980", ExternalForce="0,0,0") -> (Pose) {
    angularDampingOverride=1
    angularSpringConstant=8000
    bAngularSpring=true
    bOverrideAngularDamping=true
    bUseGravityOverride=true
    gravityScale=0.35
    simulationSpace=RootRelative
    alphaScaleBiasClamp {scale=0.5}
    boundBone {boneName=dyn_sleeve_under_r_01}
    componentAppliedLinearAccClamp=<2000 2000 2000>
    componentLinearAccScale=<0.5 0.5 0.5>
    componentLinearVelScale=<0.5 0.5 0.5>
    gravityOverride {z=-980}
    physicsBodyDefinitions [{
      boundBone {boneName=dyn_sleeve_under_r_01}
      boxExtents=<20 30 30>
      constraintSetup {angularLimitsMax=<0 10 5>, angularLimitsMin=<0 -30 -5>, angularTarget=<1 0 0>}
      localJointOffset=<4 0 0>
     }]
    simSpaceSettings {maxAngularAcceleration=40, maxAngularVelocity=8, simSpaceAngularAlpha=0.5}
   }
   n9 = Anim Dynamics - Bone: dyn_sleeve_back_r_01!(ComponentPose=@n8, Alpha=1, GravityOverride:Vector="0,0,-980", ExternalForce="0,0,0") -> (Pose) {
    angularDampingOverride=1
    angularSpringConstant=8000
    bAngularSpring=true
    bOverrideAngularDamping=true
    bUseGravityOverride=true
    gravityScale=0.35
    simulationSpace=RootRelative
    alphaScaleBiasClamp {scale=0.5}
    boundBone {boneName=dyn_sleeve_back_r_01}
    componentAppliedLinearAccClamp=<2000 2000 2000>
    componentLinearAccScale=<0.5 0.5 0.5>
    componentLinearVelScale=<0.5 0.5 0.5>
    gravityOverride {z=-980}
    physicsBodyDefinitions [{
      boundBone {boneName=dyn_sleeve_back_r_01}
      boxExtents=<20 30 30>
      constraintSetup {angularLimitsMax=<0 10 5>, angularLimitsMin=<0 -30 -5>, angularTarget=<1 0 0>}
      localJointOffset=<4 0 0>
     }]
    simSpaceSettings {maxAngularAcceleration=40, maxAngularVelocity=8, simSpaceAngularAlpha=0.5}
   }
   n10 = Anim Dynamics - Bone: dyn_boot_strap_l_01!(ComponentPose=@n9, Alpha=1, GravityOverride:Vector="0,0,-980", ExternalForce="0,0,0") -> (Pose) {
    angularDampingOverride=1
    angularSpringConstant=18000
    bAngularSpring=true
    bOverrideAngularDamping=true
    bUseGravityOverride=true
    gravityScale=0.35
    simulationSpace=RootRelative
    boundBone {boneName=dyn_boot_strap_l_01}
    componentAppliedLinearAccClamp=<2000 2000 2000>
    componentLinearAccScale=<0.6 0.6 0.6>
    componentLinearVelScale=<0.6 0.6 0.6>
    gravityOverride {z=-980}
    physicsBodyDefinitions [{
      boundBone {boneName=dyn_boot_strap_l_01}
      boxExtents=<20 30 30>
      constraintSetup {angularLimitsMax=<0 50 5>, angularLimitsMin=<0 -50 -5>, angularTarget=<1 0 0>}
      localJointOffset=<2 0 0>
     }]
    simSpaceSettings {maxAngularAcceleration=40, maxAngularVelocity=8, simSpaceAngularAlpha=0.5}
   }
   n11 = Anim Dynamics - Bone: dyn_boot_strap_r_01!(ComponentPose=@n10, Alpha=1, GravityOverride:Vector="0,0,-980", ExternalForce="0,0,0") -> (Pose) {
    angularDampingOverride=1
    angularSpringConstant=18000
    bAngularSpring=true
    bOverrideAngularDamping=true
    bUseGravityOverride=true
    gravityScale=0.35
    simulationSpace=RootRelative
    boundBone {boneName=dyn_boot_strap_r_01}
    componentAppliedLinearAccClamp=<2000 2000 2000>
    componentLinearAccScale=<0.6 0.6 0.6>
    componentLinearVelScale=<0.6 0.6 0.6>
    gravityOverride {z=-980}
    physicsBodyDefinitions [{
      boundBone {boneName=dyn_boot_strap_r_01}
      boxExtents=<20 30 30>
      constraintSetup {angularLimitsMax=<0 50 5>, angularLimitsMin=<0 -50 -5>, angularTarget=<1 0 0>}
      localJointOffset=<2 0 0>
     }]
    simSpaceSettings {maxAngularAcceleration=40, maxAngularVelocity=8, simSpaceAngularAlpha=0.5}
   }
   n12 = Anim Dynamics - Bone: dyn_boot_zipper_l_01!(ComponentPose=@n11, Alpha=1, GravityOverride:Vector="0,0,-980", ExternalForce="0,0,0") -> (Pose) {
    angularDampingOverride=1
    angularSpringConstant=20000
    bAngularSpring=true
    bOverrideAngularDamping=true
    bUseGravityOverride=true
    gravityScale=0.35
    simulationSpace=RootRelative
    boundBone {boneName=dyn_boot_zipper_l_01}
    componentAppliedLinearAccClamp=<2000 2000 2000>
    componentLinearAccScale=<0.5 0.5 0.5>
    componentLinearVelScale=<0.5 0.5 0.5>
    gravityOverride {z=-980}
    physicsBodyDefinitions [{
      boundBone {boneName=dyn_boot_zipper_l_01}
      boxExtents=<20 30 30>
      constraintSetup {angularLimitsMax=<0 10 60>, angularLimitsMin=<0 -40 -60>, angularTarget=<1 0 0>}
      localJointOffset=<10 0 0>
     }]
    simSpaceSettings {maxAngularAcceleration=40, maxAngularVelocity=8, simSpaceAngularAlpha=0.5}
   }
   n13 = Anim Dynamics - Bone: dyn_boot_zipper_r_01!(ComponentPose=@n12, Alpha=1, GravityOverride:Vector="0,0,-980", ExternalForce="0,0,0") -> (Pose) {
    angularDampingOverride=1
    angularSpringConstant=20000
    bAngularSpring=true
    bOverrideAngularDamping=true
    bUseGravityOverride=true
    gravityScale=0.35
    simulationSpace=RootRelative
    boundBone {boneName=dyn_boot_zipper_r_01}
    componentAppliedLinearAccClamp=<2000 2000 2000>
    componentLinearAccScale=<0.5 0.5 0.5>
    componentLinearVelScale=<0.5 0.5 0.5>
    gravityOverride {z=-980}
    physicsBodyDefinitions [{
      boundBone {boneName=dyn_boot_zipper_r_01}
      boxExtents=<20 30 30>
      constraintSetup {angularLimitsMax=<0 10 40>, angularLimitsMin=<0 -40 -40>, angularTarget=<1 0 0>}
      localJointOffset=<10 0 0>
     }]
    simSpaceSettings {maxAngularAcceleration=40, maxAngularVelocity=8, simSpaceAngularAlpha=0.5}
   }
   n14 = Anim Dynamics - Bone: dyn_Tshirt_fr_c_01!(ComponentPose=@n13, Alpha=1, GravityOverride:Vector="0,0,-980", ExternalForce="0,0,0") -> (Pose) {
    angularDampingOverride=1
    angularSpringConstant=12000
    bAngularSpring=true
    bOverrideAngularDamping=true
    bUseGravityOverride=true
    bUseSphericalLimits=true
    gravityScale=0.35
    simulationSpace=RootRelative
    boundBone {boneName=dyn_Tshirt_fr_c_01}
    componentAppliedLinearAccClamp=<3000 3000 3000>
    componentLinearAccScale=<0.5 0.5 0.5>
    componentLinearVelScale=<0.5 0.5 0.5>
    gravityOverride {z=-980}
    physicsBodyDefinitions [{
      boundBone {boneName=dyn_Tshirt_fr_c_01}
      boxExtents=<20 20 20>
      constraintSetup {angularLimitsMax=<0 10 5>, angularLimitsMin=<0 -10 -5>, angularTarget=<1 0 0>}
      localJointOffset=<4 0 0>
     }]
    simSpaceSettings {maxAngularAcceleration=40, maxAngularVelocity=8, simSpaceAngularAlpha=0.35}
   }
   n15 = Anim Dynamics - Bone: dyn_Tshirt_bk_c_01!(ComponentPose=@n14, Alpha=1, GravityOverride:Vector="0,0,-980", ExternalForce="0,0,0") -> (Pose) {
    angularDampingOverride=1
    angularSpringConstant=12000
    bAngularSpring=true
    bOverrideAngularDamping=true
    bUseGravityOverride=true
    bUseSphericalLimits=true
    gravityScale=0.35
    simulationSpace=RootRelative
    boundBone {boneName=dyn_Tshirt_bk_c_01}
    componentAppliedLinearAccClamp=<3000 3000 3000>
    componentLinearAccScale=<0.5 0.5 0.5>
    componentLinearVelScale=<0.5 0.5 0.5>
    gravityOverride {z=-980}
    physicsBodyDefinitions [{
      boundBone {boneName=dyn_Tshirt_bk_c_01}
      boxExtents=<20 20 20>
      constraintSetup {angularLimitsMax=<0 10 5>, angularLimitsMin=<0 -10 -5>, angularTarget=<1 0 0>}
      localJointOffset=<4 0 0>
     }]
    simSpaceSettings {maxAngularAcceleration=40, maxAngularVelocity=8, simSpaceAngularAlpha=0.35}
   }
   n16 = Anim Dynamics - Bone: dyn_Tshirt_side_l_01!(ComponentPose=@n15, Alpha=1, GravityOverride:Vector="0,0,-980", ExternalForce="0,0,0") -> (Pose) {
    angularDampingOverride=1
    angularSpringConstant=12000
    bAngularSpring=true
    bOverrideAngularDamping=true
    bUseGravityOverride=true
    bUseSphericalLimits=true
    gravityScale=0.35
    simulationSpace=RootRelative
    boundBone {boneName=dyn_Tshirt_side_l_01}
    componentAppliedLinearAccClamp=<3000 3000 3000>
    componentLinearAccScale=<0.5 0.5 0.5>
    componentLinearVelScale=<0.5 0.5 0.5>
    gravityOverride {z=-980}
    physicsBodyDefinitions [{
      boundBone {boneName=dyn_Tshirt_side_l_01}
      boxExtents=<20 20 20>
      constraintSetup {angularLimitsMax=<0 10 5>, angularLimitsMin=<0 -10 -5>, angularTarget=<1 0 0>}
      localJointOffset=<4 0 0>
     }]
    simSpaceSettings {maxAngularAcceleration=40, maxAngularVelocity=8, simSpaceAngularAlpha=0.35}
   }
   n17 = Anim Dynamics - Bone: dyn_Tshirt_side_r_01!(ComponentPose=@n16, Alpha=1, GravityOverride:Vector="0,0,-980", ExternalForce="0,0,0") -> (Pose) {
    angularDampingOverride=1
    angularSpringConstant=12000
    bAngularSpring=true
    bOverrideAngularDamping=true
    bUseGravityOverride=true
    bUseSphericalLimits=true
    gravityScale=0.35
    simulationSpace=RootRelative
    boundBone {boneName=dyn_Tshirt_side_r_01}
    componentAppliedLinearAccClamp=<3000 3000 3000>
    componentLinearAccScale=<0.5 0.5 0.5>
    componentLinearVelScale=<0.5 0.5 0.5>
    gravityOverride {z=-980}
    physicsBodyDefinitions [{
      boundBone {boneName=dyn_Tshirt_side_r_01}
      boxExtents=<20 20 20>
      constraintSetup {angularLimitsMax=<0 10 5>, angularLimitsMin=<0 -10 -5>, angularTarget=<1 0 0>}
      localJointOffset=<4 0 0>
     }]
    simSpaceSettings {maxAngularAcceleration=40, maxAngularVelocity=8, simSpaceAngularAlpha=0.35}
   }
   Output Pose(Result=Component To Local(ComponentPose=@n17)) {layerGroup=None}
  ]}
 EventGraph {nodes [
   Event BlueprintUpdateAnimation()
   > n1 = "Gravity Override State Function (Parameter Asset) (Nativized)"(
    CharacterPartAnimInstance=Self-Reference()
    GameTimeInSeconds=GetGameTimeInSeconds()
    Parameters=&/F_MED_TriadSpear_Parameters
   )
   > Set GravityOverride(GravityOverride=@n1)
   > Set RBAN Settings Out(RBAN Settings Out=@n1.OutRigidBodyAnimNodeParameters)
   > n2 = Branch(Condition=OR Boolean!(
     A=Get bIsSkyDiving()
     B=Get bIsCrouching()
     C:bool=Get bIsPlayingEmote()
     D:bool=Get bIsJumping()
     E:bool=Get bIsFalling()
     F:bool=Get bIsMoving2D()
     G:bool=Get bIsSprinting()
     H:bool=Get bIsTacticalSprinting()
     I:bool=Get bIsInFrontEnd()
    ) -> (ReturnValue))
   > Set isIdle(isIdle:double=0) -> (Output_Get:double)
   Set isIdle(execute=@n2.else, isIdle=1)
  ]}
]
variables [
 GravityOverride {type=Vector}
 isIdle {type=number}
 RBAN Settings Out {type=FortRigidBodyAnimNodeParameters}
]
